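# Niagara system for the HydraTrumpet Coach back bling: electrical spark bursts and two glowing eyes, attached through socket reads.
# Two spark emitters use separate socket ranges, one with self-timed random loops and its own color curve, the other using a system gradient.
# Spawn groups separate falling sparks from stationary flares, with flickering flare and light color and front-end dimming.
# Eye glows follow their sockets, fading with the system Timeline and time-of-day brightness.
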NiagaraSystem /BRCosmetics/Accessories/FORT_Backpacks/Backpack_F_MED_HydraTrumpet_Coach/FX/NS_Backpack_HydraTrumpet_Coach
dependencies {
 usedDataInterfaces [
  &NiagaraDataInterfaceCurve
  &NiagaraDataInterfaceColorCurve
  &NiagaraDataInterfaceArrayInt32
  &NiagaraDataInterfaceSocketReader
 ]
 usedDynamicInputs [&RandomRangeVector, &FNTimeOfDayScalar]
 usedModules [
  &self:SystemSpawnScript.SystemScriptSource.SystemScriptGraph.NiagaraNodeAssignment_3.SetVariables_#1
  &SystemState
  &Timeline
  &self:SystemSpawnScript.SystemScriptSource.SystemScriptGraph.NiagaraNodeAssignment_1.SetVariables_#2
  &self:SystemSpawnScript.SystemScriptSource.SystemScriptGraph.NiagaraNodeAssignment_0.SetVariables_#3
 ]
}
emitters {
 ElectricalSparksBurst {
  emitterUpdateScript [
   /SetVariables(Emitter.RandomSockets:NiagaraInt32=&RandomRangeInt(Minimum=5, Maximum=9)) {isSetParameters=true, module=SetVariables_#4}
   &EmitterState()
   &SpawnBurst_Instantaneous(
    Spawn Count=&RandomRangeInt(Minimum=6, Maximum=9)
    Use Spawn Probability=false
    UseLoopCountLimit=false
   )
   &SpawnBurst_Instantaneous(Spawn Group=1, Use Spawn Probability=true, UseLoopCountLimit=false) {module=SpawnBurst_Instantaneous001}
  ]
  particleSpawnScript [
   /SetVariables(Particles.VisibilityTag=$Emitter.SpawnGroup) {isSetParameters=true, module=SetVariables_#5}
   /SetVariables(Particles.PerParticleRandom:NiagaraFloat=&FloatFromCurve(
     FloatCurve=DataInterface() {propertyValues {Curve {keys [
         {arriveTangent=4.1785, interpMode=RCIM_Cubic, tangentMode=RCTM_Break, time=0, value=0}
         {arriveTangent=3.1227, interpMode=RCIM_Cubic, leaveTangent=0.0224, tangentMode=RCTM_Break, time=1, value=1}
        ]}}}
     CurveIndex=&RandomRangeFloat(Minimum=0, Maximum=1)
    )) {isSetParameters=true, module=SetVariables_#6}
   &InitializeParticle(
    Color Mode=::NewEnumerator0
    Lifetime=&RandomRangeFloat(Minimum=0.25, Maximum=0.5)
    Position Offset Coordinate Space=::Local
    UsePositionOffset=true
    Write Material Random=true
   )
   &SocketLocation(
    Interpolate Socket Sampling=true
    SamplePosOffset=true
    Socket Index=$Emitter.RandomSockets
    Socket Reader=$System.Socket Reader
    Socket Sampling Mode=::NewEnumerator3
    VelocitySpeedLimit=false
   )
   &AddVelocity(
    Additional Quaternion Rotation=false
    Cone Angle=&RandomRangeFloat(Minimum=25, Maximum=50)
    Cone Axis=&ReturnMeshOrientationAxis(Mesh Orientation Axis=::XAxis, Mesh Orientation Quat=$Output.SocketLocation.SampledRotation)
    Module.UseConeFalloff=false
    Module.UseVelDistribution=false
    Override Seed=false
    Rotation Mode=::NewEnumerator3
    Rotation Quaternion:Quat4f=$Particles.SkeletalMeshLocation.SampledRotation
    Velocity Mode=::NewEnumerator2
    Velocity Speed=&MaskFloatBySpawnGroup(Masked Float=&RandomRangeFloat(Minimum=150, Maximum=500), Passthrough Float=0)
   )
   &AddVelocity(
    Rotation Coordinate Space=::World
    Velocity=<0 0 1>
    Velocity Speed Scale=&RandomRangeFloat(Minimum=20, Maximum=40)
   ) {module=AddVelocity001}
   /SetVariables(
    Particles.FlareSpriteSize:Vector2f=<15 15>
    Particles.GlowSpriteSize:Vector2f=&Multiply_Vector2D_ByFloat(Float=0.6, Vector2D=&Lerp_Vector2(A=<5 10>, B=<10 20>, Alpha=$Particles.PerParticleRandom))
    Particles.SparkSpriteSize:Vector2f=&Multiply_Vector2D_ByFloat(Float=0.6, Vector2D=&Lerp_Vector2(A=<1 2>, B=<2 7>, Alpha=$Particles.PerParticleRandom))
   ) {isSetParameters=true, module=SetVariables_#7}
   &CameraOffset(Camera Offset Amount=&MaskFloatBySpawnGroup(Masked Float=0, Passthrough Float=12))
  ]
  particleUpdateScript [
   &SocketLocation(
    Position Sampling=::NewEnumerator1
    SamplePosOffset=false
    Socket Index=$Emitter.RandomSockets
    Socket Reader=$System.Socket Reader
    Socket Sampling Mode=::NewEnumerator3
    VelocitySpeedLimit=false
   ) {module=SocketLocation001}
   /SetVariables(Particles.Position_LightGlow:NiagaraPosition=$Output.SocketLocation001.SampledPosition) {isSetParameters=true, module=SetVariables_#8}
   /SetVariables(Particles.SparkSpriteColor:LinearColor=&MultiplyLinearColor(
     Linear Color=$System.SparkColorGradient
     Scale Factor=&MakeFloat4FromVectorAndFloat(
      "Vector (RGB)"=&VectorFromFloat(Value=&FN_FrontEndFloatModifier(Gameplay=1, Front End=0.4))
      "Float (Alpha)"=1
     )
    )) {isSetParameters=true, module=SetVariables_#9}
   /SetVariables(Particles.GlowSpriteColor:LinearColor=&MultiplyLinearColor(
     Linear Color=$Particles.SparkSpriteColor
     Scale Factor=&MakeFloat4FromVectorAndFloat("Vector (RGB)"=<1 1 1>, "Float (Alpha)"=0.03)
    )) {isSetParameters=true, module=SetVariables_#10}
   /SetVariables(Particles.FlareSpriteColor:LinearColor=&MultiplyLinearColor(
     Linear Color=$System.SparkColorGradient
     Scale Factor=&MakeFloat4FromVectorAndFloat(
      "Vector (RGB)"=&VectorFromFloat(Value=&FN_FrontEndFloatModifier(Gameplay=0.75, Front End=0.75))
      "Float (Alpha)"=&FloatFromCurve(
       FloatCurve=DataInterface() {propertyValues {Curve {keys [
           {tangentMode=RCTM_User, time=0, value=1}
           <at 0.0756 -0.7485>
           <at 0.1284 2.2279>
           <at 0.3098 -0.7559>
           <at 0.4525 0.9206>
           <at 0.525 0.0265>
           <at 0.6661 0.3676>
           <at 0.7436 1.6548>
           <at 0.8248 0.1441>
           <at 0.9286 0.8147>
           {tangentMode=RCTM_User, time=1, value=0}
          ]}}}
       CurveIndex=&NormalizeFloat(
        Value=&Multiply_Float(A=$Particles.NormalizedAge, B=&RandomRangeFloat(Minimum=2.5, Maximum=4))
        Min=0
        Max=1
       )
       Scale Curve=0.1
      )
     )
    )) {isSetParameters=true, module=SetVariables_#11}
   /SetVariables(Particles.LightColor=&MultiplyLinearColor(
     Linear Color=$Particles.FlareSpriteColor
     Scale Factor=&MakeFloat4FromVectorAndFloat("Vector (RGB)"=&VectorFromFloat(Value=3), "Float (Alpha)"=1)
    )) {isSetParameters=true, module=SetVariables_#12}
   &ParticleState()
   &Drag(
    Drag=&FloatFromCurve(
     FloatCurve=DataInterface() {propertyValues {Curve {keys [{interpMode=RCIM_Cubic, tangentMode=RCTM_User, time=0, value=1}, {interpMode=RCIM_Cubic, tangentMode=RCTM_User, time=1, value=0}]}}}
     Scale Curve=&Lerp_Float(A=17, B=5, Alpha=$Particles.PerParticleRandom)
    )
    Use Linear Drag=true
    Use Rotational Drag=true
   )
   &GravityForce(Gravity=&MaskVectorBySpawnGroup(Masked Vector=<0 0 -600>, Passthrough Vector=<0 0 0>))
   &InheritSourceMovement(
    Applied Position Scale=&VectorFromFloat(Value=&FloatFromCurve(FloatCurve=DataInterface() {propertyValues {Curve {
         defaultValue=1
         keys [
          {interpMode=RCIM_Cubic, tangentMode=RCTM_User, time=0, value=1}
          {arriveTangent=-1.9054, interpMode=RCIM_Cubic, leaveTangent=-1.9054, tangentMode=RCTM_User, time=1, value=0}
         ]
        }}}))
    ApplyForce=false
    ApplyPosition=true
    ApplyPositionFalloff=false
   ) {module=InheritSourceMovement001}
   &SolveForcesAndVelocity(Clamp Velocity=false, Limit Acceleration=false)
  ]
  rendererClasses [&NiagaraSpriteRendererProperties, &NiagaraLightRendererProperties]
  renderers [
   {rendererClass=&NiagaraSpriteRendererProperties}
   {rendererClass=&NiagaraSpriteRendererProperties, rendererIndex=1}
   {rendererClass=&NiagaraLightRendererProperties, rendererIndex=2}
   {rendererClass=&NiagaraSpriteRendererProperties, rendererIndex=3}
  ]
 }
 ElectricalSparksBurst001 {
  emitterUpdateScript [
   /SetVariables(Emitter.RandomSockets:NiagaraInt32=&RandomRangeInt(Minimum=0, Maximum=4)) {isSetParameters=true, module=SetVariables_#4}
   &EmitterState(
    Life Cycle Mode=::NewEnumerator1
    Loop Delay=&RandomRangeFloat(Minimum=0, Maximum=0.5)
    Loop Duration=&RandomRangeFloat(Minimum=9, Maximum=15, Evaluation Type=::NewEnumerator1)
    Recalculate Duration Each Loop=true
    UseLoopDelay=true
   )
   &SpawnBurst_Instantaneous(
    Spawn Count=&RandomRangeInt(Minimum=6, Maximum=9)
    Use Spawn Probability=false
    UseLoopCountLimit=false
   )
   &SpawnBurst_Instantaneous(Spawn Group=1, Use Spawn Probability=true, UseLoopCountLimit=false) {module=SpawnBurst_Instantaneous001}
  ]
  particleSpawnScript [
   /SetVariables(Particles.VisibilityTag=$Emitter.SpawnGroup) {isSetParameters=true, module=SetVariables_#5}
   /SetVariables(Particles.PerParticleRandom:NiagaraFloat=&FloatFromCurve(
     FloatCurve=DataInterface() {propertyValues {Curve {keys [
         {arriveTangent=4.1785, interpMode=RCIM_Cubic, tangentMode=RCTM_Break, time=0, value=0}
         {arriveTangent=3.1227, interpMode=RCIM_Cubic, leaveTangent=0.0224, tangentMode=RCTM_Break, time=1, value=1}
        ]}}}
     CurveIndex=&RandomRangeFloat(Minimum=0, Maximum=1)
    )) {isSetParameters=true, module=SetVariables_#6}
   &InitializeParticle(
    Color Mode=::NewEnumerator0
    Lifetime=&RandomRangeFloat(Minimum=0.3, Maximum=0.6)
    Position Offset Coordinate Space=::Local
    UsePositionOffset=true
    Write Material Random=true
   )
   &SocketLocation(
    Interpolate Socket Sampling=true
    SamplePosOffset=true
    Socket Index=$Emitter.RandomSockets
    Socket Reader=$System.Socket Reader
    Socket Sampling Mode=::NewEnumerator3
    VelocitySpeedLimit=false
   )
   &AddVelocity(
    Additional Quaternion Rotation=false
    Cone Angle=&RandomRangeFloat(Minimum=25, Maximum=50)
    Cone Axis=&ReturnMeshOrientationAxis(Mesh Orientation Axis=::XAxis, Mesh Orientation Quat=$Output.SocketLocation.SampledRotation)
    Module.UseConeFalloff=false
    Module.UseVelDistribution=false
    Override Seed=false
    Rotation Mode=::NewEnumerator3
    Rotation Quaternion:Quat4f=$Particles.SkeletalMeshLocation.SampledRotation
    Velocity Mode=::NewEnumerator2
    Velocity Speed=&MaskFloatBySpawnGroup(Masked Float=&RandomRangeFloat(Minimum=150, Maximum=400), Passthrough Float=0)
   )
   &AddVelocity(
    Rotation Coordinate Space=::World
    Velocity=<0 0 1>
    Velocity Speed Scale=&RandomRangeFloat(Minimum=10, Maximum=40)
   ) {module=AddVelocity001}
   /SetVariables(
    Particles.FlareSpriteSize:Vector2f=<12 12>
    Particles.GlowSpriteSize:Vector2f=&Multiply_Vector2D_ByFloat(Float=0.55, Vector2D=&Lerp_Vector2(A=<5 10>, B=<10 20>, Alpha=$Particles.PerParticleRandom))
    Particles.SparkSpriteSize:Vector2f=&Multiply_Vector2D_ByFloat(Float=0.55, Vector2D=&Lerp_Vector2(A=<1 2>, B=<2 7>, Alpha=$Particles.PerParticleRandom))
   ) {isSetParameters=true, module=SetVariables_#7}
   &CameraOffset(Camera Offset Amount=&MaskFloatBySpawnGroup(Masked Float=0, Passthrough Float=12))
  ]
  particleUpdateScript [
   &SocketLocation(
    Position Sampling=::NewEnumerator1
    SamplePosOffset=false
    Socket Index=$Emitter.RandomSockets
    Socket Reader=$System.Socket Reader
    Socket Sampling Mode=::NewEnumerator3
    VelocitySpeedLimit=false
   ) {module=SocketLocation001}
   /SetVariables(Particles.Position_LightGlow:NiagaraPosition=$Output.SocketLocation001.SampledPosition) {isSetParameters=true, module=SetVariables_#8}
   /SetVariables(Particles.SparkSpriteColor:LinearColor=&MultiplyLinearColor(
     Linear Color=&MultiplyLinearColor(
      Linear Color=&ColorFromCurve(ColorCurve=DataInterface() {propertyValues {
         AlphaCurve {keys [<at 0.01 1>,<at 0.9198 1 RCIM_Cubic>,<at 1 0>]}
         BlueCurve {keys [
           <at 0.0234 33>
           {arriveTangent=-446.0502, interpMode=RCIM_Cubic, time=0.0763, value=27.192}
           <at 0.105 1.1>
           <at 0.1736 0>
           <at 0.4906 0>
           <at 0.7101 0>
           <at 0.9394 0>
          ]}
         GreenCurve {keys [
           <at 0.0234 9.9>
           {arriveTangent=-21.5332, interpMode=RCIM_Cubic, time=0.0763, value=33}
           <at 0.105 8.36>
           <at 0.1736 2.2>
           <at 0.4906 1.0705>
           <at 0.7101 0.1833>
           <at 0.9394 0.0762>
          ]}
         RedCurve {keys [
           <at 0.0234 0>
           {arriveTangent=153.8104, interpMode=RCIM_Cubic, time=0.0763, value=1.32}
           <at 0.105 11>
           <at 0.1736 3>
           <at 0.4906 2>
           <at 0.7101 1>
           <at 0.9394 1>
          ]}
        }})
      Scale Factor=&MakeFloat4FromVectorAndFloat("Vector (RGB)"=&VectorFromFloat(Value=0.8), "Float (Alpha)"=1)
     )
     Scale Factor=&MakeFloat4FromVectorAndFloat(
      "Vector (RGB)"=&VectorFromFloat(Value=&FN_FrontEndFloatModifier(Gameplay=1, Front End=0.4))
      "Float (Alpha)"=0.5
     )
    )) {isSetParameters=true, module=SetVariables_#9}
   /SetVariables(Particles.GlowSpriteColor:LinearColor=&MultiplyLinearColor(
     Linear Color=$Particles.SparkSpriteColor
     Scale Factor=&MakeFloat4FromVectorAndFloat("Vector (RGB)"=<0.7 0.7 0.7>, "Float (Alpha)"=0.03)
    )) {isSetParameters=true, module=SetVariables_#10}
   /SetVariables(Particles.FlareSpriteColor:LinearColor=&MultiplyLinearColor(
     Linear Color=&MultiplyLinearColor(
      Linear Color=&ColorFromCurve(ColorCurve=DataInterface() {propertyValues {
         AlphaCurve {keys [<at 0.01 1>,<at 0.9198 1 RCIM_Cubic>,<at 1 0>]}
         BlueCurve {keys [
           <at 0.0234 33>
           {arriveTangent=-446.0502, interpMode=RCIM_Cubic, time=0.0763, value=27.192}
           <at 0.105 1.1>
           <at 0.1736 0>
           <at 0.4906 0>
           <at 0.7101 0>
           <at 0.9394 0>
          ]}
         GreenCurve {keys [
           <at 0.0234 9.9>
           {arriveTangent=-21.5332, interpMode=RCIM_Cubic, time=0.0763, value=33}
           <at 0.105 8.36>
           <at 0.1736 2.2>
           <at 0.4906 1.0705>
           <at 0.7101 0.1833>
           <at 0.9394 0.0762>
          ]}
         RedCurve {keys [
           <at 0.0234 0>
           {arriveTangent=153.8104, interpMode=RCIM_Cubic, time=0.0763, value=1.32}
           <at 0.105 11>
           <at 0.1736 3>
           <at 0.4906 2>
           <at 0.7101 1>
           <at 0.9394 1>
          ]}
        }})
      Scale Factor=&MakeFloat4FromVectorAndFloat("Vector (RGB)"=&VectorFromFloat(Value=0.8), "Float (Alpha)"=1)
     )
     Scale Factor=&MakeFloat4FromVectorAndFloat(
      "Vector (RGB)"=&VectorFromFloat(Value=&FN_FrontEndFloatModifier(Gameplay=0.75, Front End=0.75))
      "Float (Alpha)"=&FloatFromCurve(
       FloatCurve=DataInterface() {propertyValues {Curve {keys [
           {tangentMode=RCTM_User, time=0, value=1}
           <at 0.0756 -0.7485>
           <at 0.1284 2.2279>
           <at 0.3098 -0.7559>
           <at 0.4525 0.9206>
           <at 0.525 0.0265>
           <at 0.6661 0.3676>
           <at 0.7436 1.6548>
           <at 0.8248 0.1441>
           <at 0.9286 0.8147>
           {tangentMode=RCTM_User, time=1, value=0}
          ]}}}
       CurveIndex=&NormalizeFloat(
        Value=&Multiply_Float(A=$Particles.NormalizedAge, B=&RandomRangeFloat(Minimum=2.5, Maximum=4))
        Min=0
        Max=1
       )
       Scale Curve=0.05
      )
     )
    )) {isSetParameters=true, module=SetVariables_#11}
   /SetVariables(Particles.LightColor=&MultiplyLinearColor(
     Linear Color=$Particles.FlareSpriteColor
     Scale Factor=&MakeFloat4FromVectorAndFloat("Vector (RGB)"=&VectorFromFloat(Value=3), "Float (Alpha)"=1)
    )) {isSetParameters=true, module=SetVariables_#12}
   &ParticleState()
   &Drag(
    Drag=&FloatFromCurve(
     FloatCurve=DataInterface() {propertyValues {Curve {keys [{interpMode=RCIM_Cubic, tangentMode=RCTM_User, time=0, value=1}, {interpMode=RCIM_Cubic, tangentMode=RCTM_User, time=1, value=0}]}}}
     Scale Curve=&Lerp_Float(A=17, B=5, Alpha=$Particles.PerParticleRandom)
    )
    Use Linear Drag=true
    Use Rotational Drag=true
   )
   &GravityForce(Gravity=&MaskVectorBySpawnGroup(Masked Vector=<0 0 -980>, Passthrough Vector=<0 0 0>))
   &InheritSourceMovement(
    Applied Position Scale=&VectorFromFloat(Value=&FloatFromCurve(FloatCurve=DataInterface() {propertyValues {Curve {
         defaultValue=1
         keys [
          {interpMode=RCIM_Cubic, tangentMode=RCTM_User, time=0, value=1}
          {arriveTangent=-1.9054, interpMode=RCIM_Cubic, leaveTangent=-1.9054, tangentMode=RCTM_User, time=1, value=0}
         ]
        }}}))
    ApplyForce=false
    ApplyPosition=true
    ApplyPositionFalloff=false
   ) {module=InheritSourceMovement002}
   &SolveForcesAndVelocity(Clamp Velocity=false, Limit Acceleration=false)
  ]
  rendererClasses [&NiagaraSpriteRendererProperties, &NiagaraLightRendererProperties]
  renderers [
   {rendererClass=&NiagaraSpriteRendererProperties}
   {rendererClass=&NiagaraSpriteRendererProperties, rendererIndex=1}
   {rendererClass=&NiagaraLightRendererProperties, rendererIndex=2}
   {rendererClass=&NiagaraSpriteRendererProperties, rendererIndex=3}
  ]
 }
 EyeGlows {
  emitterUpdateScript [
   &EmitterState(
    Inactive Response=::NewEnumerator1
    Life Cycle Mode=::NewEnumerator1
    Loop Behavior=::NewEnumerator1
    Loop Duration Mode=::NewEnumerator1
    UseLoopDelay=false
   )
   &SpawnBurst_Instantaneous(Spawn Count=2, Use Spawn Probability=false, UseLoopCountLimit=false)
  ]
  particleSpawnScript [
   &PartitionParticles()
   &InitializeParticle(Color=<rgba 0 0.2384 0.2346 0.2>, Sprite Size Mode=::NewEnumerator3, UsePositionOffset=false, Write Material Random=true)
   &DynamicMaterialParameters(
    Index 0 Param 1=0.2
    Index 0 Param 2=1
    Module.Param0WriteEnabled=true
    Module.Param1WriteEnabled=true
    Module.Param2WriteEnabled=true
    Module.Param3WriteEnabled=true
   )
   &CameraOffset(Camera Offset Amount=1.7)
  ]
  particleUpdateScript [
   &ParticleState(Kill Particles When Lifetime Has Elapsed=false, Let Infinitely Lived Particles Die When Emitter Deactivates=true)
   &ScaleSpriteSize(
    Scale Sprite Size Mode=::NewEnumerator0
    Uniform Scale Factor=&FloatFromCurve(
     FloatCurve=DataInterface() {propertyValues {Curve {
        defaultValue=1
        keys [
         {arriveTangent=4.0085, interpMode=RCIM_Cubic, leaveTangent=4.0085, tangentMode=RCTM_User, time=0, value=0}
         <at 0.01 1>
         {arriveTangent=-1.0101, leaveTangent=-1.0101, time=0.99, value=1}
         {tangentMode=RCTM_User, time=1, value=0}
        ]
       }}}
     CurveIndex=$System.Timeline.NormalizedTime
    )
   )
   &SocketLocation(
    SamplePosOffset=false
    Socket Index=&SelectIntFromArray(Direct Array Index=$Particles.PartitionParticles.Partition, Int Selection Array=DataInterface() {propertyValues {IntData [10,11]}})
    Socket Reader=$System.Socket Reader
    Socket Sampling Mode=::NewEnumerator3
    VelocitySpeedLimit=false
   )
   &ScaleColor(
    Scale Alpha=&Multiply_Float(
     A=$System.TimeOfDayBrightnessMultiplier
     B=&FloatFromCurve(
      FloatCurve=DataInterface() {propertyValues {Curve {
         defaultValue=1
         keys [
          {arriveTangent=4.0085, interpMode=RCIM_Cubic, leaveTangent=4.0085, tangentMode=RCTM_User, time=0, value=0}
          {interpMode=RCIM_Cubic, tangentMode=RCTM_User, time=0.5, value=1}
          {tangentMode=RCTM_User, time=1, value=0}
         ]
        }}}
      CurveIndex=$System.Timeline.NormalizedTime
     )
    )
    ScaleA=true
    ScaleRGB=true
   )
  ]
  rendererClasses [&NiagaraSpriteRendererProperties]
  renderers [{rendererClass=&NiagaraSpriteRendererProperties}]
 }
}
summary {
 systemName=NS_Backpack_HydraTrumpet_Coach
 emitters [
  {
   emitterName=ElectricalSparksBurst001
   rendererClasses [&NiagaraSpriteRendererProperties, &NiagaraLightRendererProperties]
  }
  {
   emitterName=ElectricalSparksBurst
   rendererClasses [&NiagaraSpriteRendererProperties, &NiagaraLightRendererProperties]
  }
  {emitterName=EyeGlows, rendererClasses [&NiagaraSpriteRendererProperties]}
 ]
 userVariables [
  User.bIsFrontEnd {type=NiagaraBool}
  User.bIsFrontEndPreview {type=NiagaraBool}
  User.Reactive {type=NiagaraBool}
 ]
}
