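# Parent surface material for the HydraTrumpet Coach backpack cosmetic, built on material attributes for skeletal meshes and mesh deformers.
# The 2023 character parent PreFX function feeds a backpack-specific reactivity function, whose result enters the matching PostFX function alongside PreFX mask data.
# A separate Get/Set attributes branch off PreFX handles Roughness and Metallic, with no visible link into the PostFX chain.
# Shading model comes from the material expression.
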Material /BRCosmetics/Accessories/FORT_Backpacks/Backpack_F_MED_HydraTrumpet_Coach/Materials/M_HydraTrumpet_Backpack_Parent
graph {
 properties {
  bUseMaterialAttributes=true
  bUsedWithMeshDeformer=true
  bUsedWithSkeletalMesh=true
  pixelDepthOffsetMode=PDOM_Legacy
  shadingModel=MSM_FromMaterialExpression
  usedShadingModels=Default Lit | Subsurface | Clear Coat | Subsurface Profile | Cloth
 }
 nodes [
  n1 = MaterialFunctionCall() {materialFunction=&MF_CharacterParent_2023_PreFX}
  MaterialFunctionCall(
   Material=MaterialFunctionCall!(MaterialAttributes=@n1.Result) {materialFunction=&/MF_HydraTrumpet_Backpack_Reactivity}.Result
   M=@n1.M
   Pre Skinned Local Position=@n1.M
   Distance Blend=@n1.M
   Fresnel=@n1.M
  ) {materialFunction=&MF_CharacterParent_2023_PostFX}
  n2 = GetMaterialAttributes(None=@n1.Result) {attributeGetTypes [#1, #2]}
  SetMaterialAttributes!(MaterialAttributes=@n2.MaterialAttributes, Roughness=@n2.MaterialAttributes, Metallic=@n2.MaterialAttributes) {attributeSetTypes [#1, #2]}
 ]
}
