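# Physics asset for the HydraTrumpet Coach backpack cosmetic, covering a short right-side bone chain on the pack mesh.
# Capsule bodies sit on the pack, clavicle_rr and shoulder_01_r bones, linked in that order by two constraints.
# The clavicle joint rotates freely against the pack, while the shoulder joint has narrow swing limits and no twist range relative to the clavicle.
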PhysicsAsset /BRCosmetics/Accessories/FORT_Backpacks/Backpack_F_MED_HydraTrumpet_Coach/Meshes/F_MED_HydraTrumpet_Coach_Pack_Physics1
bodies {
 clavicle_rr [{
   shapeType=Capsule
   length=6.7938
   radius=5.7856
   shapeName=clavicle_rr_capsule
   center=<4.1249 2.1551 -1.554>
   rotation=<rot -22.5191 -90 3.8678>
  }]
 pack [{shapeType=Capsule, length=3, radius=3}]
 shoulder_01_r [{shapeType=Capsule, length=5.2782, radius=4.5415, shapeName=shoulder_01_r_capsule, center=<4.264 0 0>, rotation=<rot 90 0 0>}]
}
constraints [
 {
  bone1Name=clavicle_rr
  bone2Name=pack
  swing1LimitDegrees=45
  swing1Motion=Free
  swing2LimitDegrees=45
  swing2Motion=Free
  twistLimitDegrees=45
  twistMotion=Free
 }
 {bone1Name=shoulder_01_r, bone2Name=clavicle_rr, swing1LimitDegrees=20, swing2LimitDegrees=20, twistLimitDegrees=0}
]
