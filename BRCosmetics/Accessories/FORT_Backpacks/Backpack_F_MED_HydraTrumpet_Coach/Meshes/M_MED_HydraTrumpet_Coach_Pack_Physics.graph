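# Physics asset for the HydraTrumpet Coach backpack cosmetic, defining capsule bodies and joints for its net, torso wires and body wires.
# Each of the three chains has a base body with free angular motion on the pack and a dyn body held to that base by narrow swing limits and zero twist limit.
# Larger net_upperleg and upperleg capsules, apparently leg collision volumes, chain onto dyn_net through angularly locked constraints.
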PhysicsAsset /BRCosmetics/Accessories/FORT_Backpacks/Backpack_F_MED_HydraTrumpet_Coach/Meshes/M_MED_HydraTrumpet_Coach_Pack_Physics
bodies {
 base_body_wires [{shapeType=Capsule, length=0.505, radius=0.505, shapeName=base_body_wires_capsule, rotation=<rot 0 -90 0>}]
 base_net [{shapeType=Capsule, length=3, radius=3, shapeName=base_net_capsule}]
 base_wires_torso [{shapeType=Capsule, length=0.505, radius=0.505, shapeName=base_wires_torso_capsule, rotation=<rot 0 -90 0>}]
 dyn_body_wires [{shapeType=Capsule, length=2, radius=2, shapeName=dyn_body_wires_capsule, center=<5 0 0>, rotation=<rot 90 0 0>}]
 dyn_net [{shapeType=Capsule, length=5.0487, radius=7.2439, center=<24.2631 1.497 -24.0979>, rotation=<rot 49.9999 0 0>}]
 dyn_wires_torso [{shapeType=Capsule, length=2, radius=3, shapeName=dyn_wires_torso_capsule, center=<10.3221 0 0>, rotation=<rot 90 0 0>}]
 net_upperleg [{
   shapeType=Capsule
   length=16.6762
   radius=8.7087
   shapeName=net_upperleg_capsule
   center=<16.4925 -0.8573 0.0087>
   rotation=<rot 0 91.5552 -77.3346>
  }]
 pack [{shapeType=Capsule, length=3, radius=3}]
 upperleg [{
   shapeType=Capsule
   length=21.8615
   radius=6.3231
   shapeName=upperleg_capsule
   center=<21.4817 0.1933 2.7027>
   rotation=<rot 0 91.439 -81.7126>
  }]
}
constraints [
 {bone1Name=dyn_net, bone2Name=base_net, swing1LimitDegrees=5, swing2LimitDegrees=10, twistLimitDegrees=0}
 {
  bone1Name=base_net
  bone2Name=pack
  swing1LimitDegrees=45
  swing1Motion=Free
  swing2LimitDegrees=45
  swing2Motion=Free
  twistLimitDegrees=45
  twistMotion=Free
 }
 {bone1Name=dyn_wires_torso, bone2Name=base_wires_torso, swing1LimitDegrees=15, swing2LimitDegrees=15, twistLimitDegrees=0}
 {
  bone1Name=base_wires_torso
  bone2Name=pack
  swing1LimitDegrees=45
  swing1Motion=Free
  swing2LimitDegrees=45
  swing2Motion=Free
  twistLimitDegrees=45
  twistMotion=Free
 }
 {
  bone1Name=net_upperleg
  bone2Name=dyn_net
  swing1LimitDegrees=45
  swing1Motion=Locked
  swing2LimitDegrees=45
  swing2Motion=Locked
  twistLimitDegrees=45
  twistMotion=Locked
 }
 {
  bone1Name=upperleg
  bone2Name=net_upperleg
  swing1LimitDegrees=45
  swing1Motion=Locked
  swing2LimitDegrees=45
  swing2Motion=Locked
  twistLimitDegrees=45
  twistMotion=Locked
 }
 {bone1Name=dyn_body_wires, bone2Name=base_body_wires, swing1LimitDegrees=10, swing2LimitDegrees=7, twistLimitDegrees=0}
 {
  bone1Name=base_body_wires
  bone2Name=pack
  swing1LimitDegrees=45
  swing1Motion=Free
  swing2LimitDegrees=45
  swing2Motion=Free
  twistLimitDegrees=45
  twistMotion=Free
 }
]
